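# Animation Blueprint for the F_MED_TealMink_Puff female character body, following the attached parent mesh and adding secondary motion.
# The copied parent pose feeds a Control Rig, then chained AnimDynamics nodes with angular springs and limits for two necklace bones and left and right zippers, all sharing one Alpha.
# The zippers also use a gravity override.
# A final additive bone transform nudges attach_backpack only when bIsInFrontEnd is true.
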AnimBlueprint /BRCosmetics/Characters/Player/Female/Medium/Bodies/F_MED_TealMink_Puff/Meshes/F_MED_TealMink_Puff_AnimBP
target_skeleton=&/F_MED_TealMink_Puff_Skeleton
graphs [
 AnimGraph {nodes [
   n1 = Get Alpha() -> (Alpha:float)
   n2 = Control Rig(Source=Copy Pose From Mesh() {bUseAttachedParent=true}, Alpha=@n1) {
    alphaBoolBlend {blendOption=Linear}
    controlRigAssetReference {blueprintRigClass=&/F_MED_TealMink_Puff_CtrlRig.F_MED_TealMink_Puff_CtrlRig_C}
   }
   n3 = Anim Dynamics - Bone: dyn_necklace_1(ComponentPose=Local To Component(LocalPose=@n2), Alpha=@n1, ExternalForce="0,0,0") {
    angularDampingOverride=0.9999
    angularSpringConstant=500
    bAngularSpring=true
    bGravityOverrideInSimSpace=true
    bOverrideAngularDamping=true
    simulationSpace=BoneRelative
    alphaBoolBlend {blendOption=Linear}
    boundBone {boneName=dyn_necklace_1}
    componentAppliedLinearAccClamp=<3500 3500 3500>
    componentLinearAccScale=<1.2 1.2 1.2>
    componentLinearVelScale=<1.2 1.2 1.2>
    physicsBodyDefinitions [{
      sphereCollisionRadius=0
      boundBone {boneName=dyn_necklace_1}
      boxExtents=<15 20 20>
      constraintSetup {angularLimitsMax=<1 30 5>, angularLimitsMin=<-1 0 -5>, angularTarget=<0 0 0>}
      localJointOffset=<4 0 0>
     }]
    relativeSpaceBone {boneName=root}
   }
   n4 = Anim Dynamics - Bone: dyn_necklace_2(ComponentPose=@n3, Alpha=@n1, ExternalForce="0,0,0") {
    angularDampingOverride=1
    angularSpringConstant=500
    bAngularSpring=true
    bGravityOverrideInSimSpace=true
    bOverrideAngularDamping=true
    simulationSpace=BoneRelative
    alphaBoolBlend {blendOption=Linear}
    boundBone {boneName=dyn_necklace_2}
    componentAppliedLinearAccClamp=<3500 3500 3500>
    componentLinearAccScale=<1.2 1.2 1.2>
    componentLinearVelScale=<1.2 1.2 1.2>
    physicsBodyDefinitions [{
      sphereCollisionRadius=0
      boundBone {boneName=dyn_necklace_2}
      boxExtents=<15 20 20>
      constraintSetup {angularLimitsMax=<1 30 10>, angularLimitsMin=<-1 0 -10>, angularTarget=<0 0 0>}
      localJointOffset=<4 0 0>
     }]
    relativeSpaceBone {boneName=root}
   }
   n5 = Anim Dynamics - Bone: dyn_zipper_l(ComponentPose=@n4, Alpha=@n1, ExternalForce="0,0,0") {
    angularDampingOverride=0.995
    angularSpringConstant=10
    bAngularSpring=true
    bGravityOverrideInSimSpace=true
    bOverrideAngularDamping=true
    bUseGravityOverride=true
    simulationSpace=RootRelative
    alphaBoolBlend {blendOption=Linear}
    boundBone {boneName=dyn_zipper_l}
    componentAppliedLinearAccClamp=<500 500 500>
    componentLinearAccScale=<1 1 1>
    componentLinearVelScale=<1 1 1>
    gravityOverride {z=-980}
    physicsBodyDefinitions [{
      boundBone {boneName=dyn_zipper_l}
      boxExtents=<4 4 10>
      constraintSetup {angularLimitsMax=<1 15 1>, angularLimitsMin=<-60 -15 -1>, angularTarget=<0 0 0>}
      localJointOffset=<0 0 -1>
     }]
    relativeSpaceBone {boneName=thigh_l}
   }
   n6 = Anim Dynamics - Bone: dyn_zipper_r(ComponentPose=@n5, Alpha=@n1, ExternalForce="0,0,0") {
    angularDampingOverride=0.995
    angularSpringConstant=10
    bAngularSpring=true
    bGravityOverrideInSimSpace=true
    bOverrideAngularDamping=true
    bUseGravityOverride=true
    simulationSpace=RootRelative
    alphaBoolBlend {blendOption=Linear}
    boundBone {boneName=dyn_zipper_r}
    componentAppliedLinearAccClamp=<500 500 500>
    componentLinearAccScale=<1 1 1>
    componentLinearVelScale=<1 1 1>
    gravityOverride {z=-980}
    physicsBodyDefinitions [{
      boundBone {boneName=dyn_zipper_r}
      boxExtents=<4 4 10>
      constraintSetup {angularLimitsMax=<1 15 1>, angularLimitsMin=<-60 -15 -1>, angularTarget=<0 0 0>}
      localJointOffset=<0 0 -1>
     }]
    relativeSpaceBone {boneName=thigh_l}
   }
   n7 = "Transform (Modify) Bone - Bone: attach_backpack"!(ComponentPose=@n6, bAlphaBoolEnabled:bool=Get bIsInFrontEnd(), Translation="0,0.5,0", Rotation="0,0,0", Scale="1,1,1") -> (Pose) {
    alphaInputType=Bool
    rotationMode=BMM_Additive
    rotationSpace=BCS_BoneSpace
    translationMode=BMM_Additive
    translationSpace=BCS_BoneSpace
    alphaBoolBlend {blendOption=Linear}
    alphaScaleBiasClamp {bInterpResult=true}
    boneToModify {boneName=attach_backpack}
    translation {y=0.5}
   }
   Output Pose(Result=Component To Local(ComponentPose=@n7)) {layerGroup=None}
  ]}
 EventGraph {nodes [Event BlueprintUpdateAnimation()]}
]
variables [Alpha {type=number, default=1}]
